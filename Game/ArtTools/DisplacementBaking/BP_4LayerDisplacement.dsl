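; Generated dynamic mesh actor that displaces a source static mesh with up to four texture layers, with a function to bake a "_Displ" static mesh asset.
; On rebuild, it extracts a masked guide submesh and packs its normals, matched by position, into vertex colors as displacement directions.
; It then tessellates the whole mesh or only displaced triangles.
; Each vertex moves by its UV0-sampled map, chosen by floored UV1 X, with per-layer magnitude and bias.
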(event Events|EventOnRebuildGeneratedMesh
  (params (TargetMesh :type Dynamic%20Mesh%20Object%20Reference))
  (bind _tessellationlevel (Variables|Default|GetTessellationLevel))
  (bind _targetmesh (Variables|Default|GetTargetMesh))
  (Utilities|FlowControl|Sequence
    (:then_0
      (bind _output_get (Variables|Default|SetTargetMesh TargetMesh))
      (bind _sourceasset (Variables|Default|GetSourceAsset)
        (:then
          (GeometryScript|StaticMesh|CopyMeshFromStaticMeshwithSectionMaterials _sourceasset _output_get (struct :bApplyBuildSettings true :bRequestTangents true :bIgnoreRemoveDegenerates true :bUseBuildScale true) (struct :LODType "RenderData" :LODIndex 0)
            (:Success
              (if (Variables|Default|GetEnabled)
                (Utilities|FlowControl|Sequence
                  (:then_0
                    (Utilities|FlowControl|Sequence
                      (:then_0
                        ;[1 1 1] Make Triangle list for Guide Mesh/ uses red vertex color for a mask what gets displaced and what not
                        (bind _uvlist (GeometryScript|UVs|GetMeshPerVertexUVs (Variables|Default|GetTargetMesh) 1))
                        (bind _triangleidlist (GeometryScript|MeshQueries|GetAllTriangleIDs (Variables|Default|GetTargetMesh)))
                        (Variables|PreTesselationData|SetTriangleIDList_PreTesselation _triangleidlist)
                        (bind _trianglelist (GeometryScript|MeshQueries|GetAllTriangleIndices (Variables|Default|GetTargetMesh)))
                        (if (Variables|Options|GetUseUV1Mask)
                          ;[1 1 1] Mask Using UV1 Y coordinate - currently used
                          (CallFunction|MakeGuideMeshTriangleListfromUV1Mask _trianglelist _uvlist)
                          (else
                            ;[1 0.1294 0.05967] using vertex color will propably not be used
                            (CallFunction|MakeGuideMeshTriangleListfromVertexColorMask _trianglelist))))
                      (:then_1
                        (bind _returnvalue (DynamicMeshActor|AllocateComputeMesh))
                        (bind _storetosubmeshout (GeometryScript|Decomposition|GetSubMeshfromMesh (Variables|Default|GetTargetMesh) _returnvalue (Variables|GuideMeshData|GetGuideMeshTriangleList)))
                        (Variables|GuideMeshData|SetGuideMesh _storetosubmeshout)
                        (Utilities|FlowControl|Sequence
                          (:then_0
                            ;[1 1 1] Make array with all vertex positions and list with all Normals from GuideMesh
                            (CallFunction|MakeGuideMeshDataLists))
                          (:then_1
                            ;[1 1 1] Loop over each vert and find matching vert on guide mesh'
                            (CallFunction|StealNormal))
                          (:then_2
                            ;[1 1 1] Write new Vertex normals. if we end up not using the houdini script at all. we could optmize this and store it not biased sacle in array and just use that later
                            (bind _colorlist (GeometryScript|ListUtils|ConvertArraytoColorList (Variables|PreTesselationData|GetColor%20Array)))
                            (GeometryScript|VertexColor|SetMeshPerVertexColors (Variables|Default|GetTargetMesh) _colorlist))))))
                  (:then_1
                    (if (Variables|Options|GetSelectiveTesselation)
                      ;[1 1 1] Split Mesh and only tesselate whats needed
                      (Utilities|FlowControl|Sequence
                        (:then_0
                          (CallFunction|MakeTriangleListforTesselation))
                        (:then_1
                          (CallFunction|TesselationfromTriangleList)))
                      (elif (Variables|Options|GetPNTesseleation)
                        ;[1 1 1] Tesselation of entire mesh. slow with meshes that have detail like window frames and doors
                        (GeometryScript|Subdivide|ApplyPNTessellation _targetmesh (struct :bRecomputeNormals false) _tessellationlevel)
                        (else
                          (GeometryScript|Subdivide|ApplyUniformTessellation _targetmesh _tessellationlevel))))
                    (Utilities|FlowControl|Sequence
                      (:then_0
                        (bind _positionlist (GeometryScript|MeshQueries|GetAllVertexPositions (Variables|Default|GetTargetMesh)))
                        (Variables|Default|SetVertexPositions _positionlist))
                      (:then_1
                        (bind _uvlist_2 (GeometryScript|UVs|GetMeshPerVertexUVs (Variables|Default|GetTargetMesh)))
                        (Variables|Default|SetUV0 _uvlist_2))
                      (:then_2
                        (bind _uvlist_3 (GeometryScript|UVs|GetMeshPerVertexUVs (Variables|Default|GetTargetMesh) 1))
                        (Variables|Default|SetUV1 _uvlist_3))
                      (:then_3
                        (bind _colorlist_2 (GeometryScript|VertexColor|GetMeshPerVertexColors (Variables|Default|GetTargetMesh)))
                        (Variables|Default|SetVertexColors _colorlist_2))
                      (:then_4
                        (bind _normallist (GeometryScript|Normals|GetMeshPerVertexNormals (Variables|Default|GetTargetMesh)))
                        (bind _output_get_2 (Variables|Default|SetVertexNormals _normallist))
                        (bind _duplicatelist (GeometryScript|ListUtils|DuplicateVectorList _output_get_2))
                        (Variables|Default|SetDisplacementVectors _duplicatelist))
                      (:then_5
                        (bind _displacementmap_1 (Variables|Default|GetDisplacementMap_1)
                          (:then
                            (bind _colorlist_3 (GeometryScript|TextureUtils|SampleTexture2DAtUVPositions (Variables|Default|GetUV0) _displacementmap_1 (struct :SamplingMethod "Bilinear" :bWrap true :UVScale "(X=1.000000,Y=1.000000)" :UVOffset "(X=0.000000,Y=0.000000)")))
                            (Variables|Default|SetTexture0Values _colorlist_3))
                          (:else)))
                      (:then_6
                        (bind _displacementmap_2 (Variables|Default|GetDisplacementMap_2)
                          (:then
                            (bind _colorlist_4 (GeometryScript|TextureUtils|SampleTexture2DAtUVPositions (Variables|Default|GetUV0) _displacementmap_2 (struct :SamplingMethod "Bilinear" :bWrap true :UVScale "(X=1.000000,Y=1.000000)" :UVOffset "(X=0.000000,Y=0.000000)")))
                            (Variables|Default|SetTexture%201Values _colorlist_4))
                          (:else)))
                      (:then_7
                        (bind _displacementmap_3 (Variables|Default|GetDisplacementMap_3)
                          (:then
                            (bind _colorlist_5 (GeometryScript|TextureUtils|SampleTexture2DAtUVPositions (Variables|Default|GetUV0) _displacementmap_3 (struct :SamplingMethod "Bilinear" :bWrap true :UVScale "(X=1.000000,Y=1.000000)" :UVOffset "(X=0.000000,Y=0.000000)")))
                            (Variables|Default|SetTexture%202Values _colorlist_5))
                          (:else)))
                      (:then_8
                        (bind _displacementmap_4 (Variables|Default|GetDisplacementMap_4)
                          (:then
                            (bind _colorlist_6 (GeometryScript|TextureUtils|SampleTexture2DAtUVPositions (Variables|Default|GetUV0) _displacementmap_4 (struct :SamplingMethod "Bilinear" :bWrap true :UVScale "(X=1.000000,Y=1.000000)" :UVOffset "(X=0.000000,Y=0.000000)")))
                            (Variables|Default|SetTexture%203Values _colorlist_6))
                          (:else)))
                      (:then_10
                        (bind _output_get_3 (Variables|Default|SetVertexCount (GeometryScript|ListUtils|GetVectorListLength (Variables|Default|GetVertexPositions))))
                        (for _index (range _output_get_3)
                          (bind _magnitude (CallFunction|GetDisplaceMapValue (Variables|Default|GetUV1) (Variables|Default|GetTexture0Values) _index (Variables|Default|GetTexture%201Values) (Variables|Default|GetTexture%202Values) (Variables|Default|GetTexture%203Values)))
                          (bind _displacement (CallFunction|MakeDisplaceVector (Math|Vector|Normalize (* (- (Math|Conversions|ToVector(LinearColor) (.ReturnValue (GeometryScript|ListUtils|GetColorListItem (Variables|Default|GetVertexColors) _index))) "0.500000,0.500000,0.500000") 2.0)) (.ReturnValue (GeometryScript|ListUtils|GetColorListItem (Variables|Default|GetVertexColors) _index)) _magnitude))
                          (GeometryScript|MeshEdits|SetVertexPosition (Variables|Default|GetTargetMesh) _index (+ _displacement (.ReturnValue (GeometryScript|ListUtils|GetVectorListItem (Variables|Default|GetVertexPositions) _index))) true)))
                      (:then_11
                        ;[1 1 1] Comment
                        (bind _returnvalue_2 (AddComponent|NODE_AddTextRenderComponent-2 :RelativeTransform (Math|Transform|MakeTransform "256.000000, 0,512.000000" "0,90.000000, 0")))
                        (Rendering|Components|TextRender|SetText _returnvalue_2 (Utilities|Text|ToText(String) (Utilities|String|Append "Vertex Count = " (Utilities|String|ToString(Integer) (Variables|Default|GetVertexCount))))))
                      (:then_12
                        ;[1 0.7842 0.1087] Debug GuideMesh
                        (if (Variables|Options|GetDebugGuideMesh)
                          (GeometryScript|MeshEdits|AppendMesh (Variables|Default|GetTargetMesh) (Variables|GuideMeshData|GetGuideMesh) (Math|Transform|MakeTransform "200.000000,200.000000, 0") false "(CombineMode=EnableAllMatching)"))))))))))
        (:else)))))

(fn BakeStaticMesh
  (bind _sourceasset (Variables|Default|GetSourceAsset))
  (bind _returnvalue (Utilities|GetObjectPathString _sourceasset))
  (bind _returnvalue_2 (Utilities|String|FindSubstring _returnvalue "."))
  (bind _returnvalue_3 (Utilities|String|GetSubstring _returnvalue 0 _returnvalue_2))
  (bind _returnvalue_4 (Utilities|String|Append _returnvalue_3 "_Displ"))
  (EditorScripting|Asset|GetPathNameforLoadedAsset _sourceasset :!class /Script/EditorScriptingUtilities.EditorAssetLibrary)
  (Development|PrintString _returnvalue_4)
  (GeometryScript|AssetManagement|CreateNewStaticMeshAssetfromMesh (Variables|Default|GetTargetMesh) _returnvalue_4))

(fn FindlNormalFromGuidemesh
  (params (ItemToFind :type Vector%20%28by%20ref%29))
  (bind _tolerance (Variables|Default|GetTolerance))
  (bind _returnvalue (Math|Vector|MakeVector _tolerance _tolerance _tolerance))
  (bind _returnvalue_2 (/ ItemToFind _returnvalue))
  (bind _z (.z _returnvalue_2))
  (bind _returnvalue_3 (Math|Float|Ceil _z))
  (bind _returnvalue_4 (Math|Conversions|ToFloat(Integer) _returnvalue_3))
  (bind _y (.y _returnvalue_2))
  (bind _returnvalue_5 (Math|Float|Ceil _y))
  (bind _returnvalue_6 (Math|Conversions|ToFloat(Integer) _returnvalue_5))
  (bind _x (.x _returnvalue_2))
  (bind _returnvalue_7 (Math|Float|Ceil _x))
  (bind _returnvalue_8 (Math|Conversions|ToFloat(Integer) _returnvalue_7))
  (bind _returnvalue_9 (Math|Vector|MakeVector _returnvalue_8 _returnvalue_6 _returnvalue_4))
  (bind _returnvalue_10 (* _returnvalue_9 _returnvalue))
  (bind _guidemeshposition (Variables|GuideMeshData|GetGuideMeshPosition))
  (bind _returnvalue_11 (Utilities|Array|FindItem _guidemeshposition _returnvalue_10))
  (bind (_bisvalidindex _returnvalue_12) (GeometryScript|ListUtils|GetVectorListItem (Variables|GuideMeshData|GetGuideMeshNormalList) _returnvalue_11))
  (bind _color_array (Variables|PreTesselationData|GetColor%20Array))
  (if (>= _returnvalue_11 0)
    (Utilities|Array|Add _color_array (Math|Conversions|ToLinearColor(Vector) (/ (+ _returnvalue_12 "1.000000,1.000000,1.000000") 2.0)))
    (else
      (Utilities|Array|Add _color_array "(R=0.500000,G=0.500000,B=0.500000,A=1.000000)"))))

(fn FindNormalFromGuidemeshMultipleSamples
  (params (Position :type Vector))
  (bind _newnormalarray (Variables|GuideMeshData|GetNewNormalArray))
  (Utilities|Array|Clear (Variables|GuideMeshData|GetNewNormalArray))
  (for (_array_element _array_index) (Variables|GuideMeshData|GetGuideMeshPosition)
    (if (< (Math|Vector|Distance(Vector) Position _array_element) 0.1)
      (Utilities|Array|Add (Variables|GuideMeshData|GetNewNormalArray) (.ReturnValue (GeometryScript|ListUtils|GetVectorListItem (Variables|GuideMeshData|GetGuideMeshNormalList) _array_index)))))
  (if (> (Utilities|Array|Length (Variables|GuideMeshData|GetNewNormalArray)) 0)
    (Variables|GuideMeshData|SetNewNormal "0, 0, 0")
    (for _array_element_2 _newnormalarray
      (Variables|GuideMeshData|SetNewNormal (+ (Variables|GuideMeshData|GetNewNormal) _array_element_2)))
    (bind _output_get (Variables|GuideMeshData|SetNewNormal (Math|Vector|Normalize (/ (Variables|GuideMeshData|GetNewNormal) (Utilities|Array|Length _newnormalarray)))))
    (Utilities|Array|Add (Variables|PreTesselationData|GetColor%20Array) (Math|Conversions|ToLinearColor(Vector) (/ (+ _output_get "1.000000,1.000000,1.000000") 2.0)))
    (else
      (Utilities|Array|Add (Variables|PreTesselationData|GetColor%20Array) "(R=0.500000,G=0.500000,B=0.500000,A=1.000000)"))))

(fn GetDisplaceMapValue
  (params (UV1 :type UV%20List%20Structure) (DisplaceMap0 :type Color%20List%20Structure) (Index :type Integer) (DisplaceMap1 :type Color%20List%20Structure) (DisplaceMap2 :type Color%20List%20Structure) (DisplaceMap3 :type Color%20List%20Structure))
  (bind (_bisvalidindex _returnvalue) (GeometryScript|ListUtils|GetColorListItem DisplaceMap0 Index))
  (bind (_bisvalidindex_2 _returnvalue_2) (GeometryScript|ListUtils|GetColorListItem DisplaceMap3 Index))
  (bind (_bisvalidindex_3 _returnvalue_3) (GeometryScript|ListUtils|GetColorListItem DisplaceMap1 Index))
  (bind (_bisvalidindex_4 _returnvalue_4) (GeometryScript|ListUtils|GetColorListItem DisplaceMap2 Index))
  (bind (_bisvalidindex_5 _returnvalue_5) (GeometryScript|ListUtils|GetUVListItem UV1 Index))
  (switch int (Math|Float|Floor (Math|Vector2D|BreakVector2D _returnvalue_5))
    (:Default
      _)
    (:0
      (if _bisvalidindex
        (return (+ (* (Math|Color|BreakColor _returnvalue) (Variables|Default|GetMagnitude_1)) (Variables|Default|GetBias_1)))
        (else
          _)))
    (:1
      (if _bisvalidindex_3
        (return (+ (* (Math|Color|BreakColor _returnvalue_3) (Variables|Default|GetMagnitude_2)) (Variables|Default|GetBias_2)))
        (else
          _)))
    (:2
      (if _bisvalidindex_4
        (return (+ (* (Math|Color|BreakColor _returnvalue_4) (Variables|Default|GetMagnitude_3)) (Variables|Default|GetBias_3)))
        (else
          _)))
    (:3
      (if _bisvalidindex_2
        (return (+ (* (Math|Color|BreakColor _returnvalue_2) (Variables|Default|GetMagnitude_4)) (Variables|Default|GetBias_4)))
        (else
          _))))
  (return 0.0))

(fn GridSliceMesh
  (params (Mesh :type Dynamic%20Mesh%20Object%20Reference) (Number%20of%20Cuts%20X :type Integer) (Number%20of%20Cuts%20Y :type Integer))
  (bind (_min _max) (Utilities|Struct|BreakBox (GeometryScript|MeshQueries|GetMeshBoundingBox (Variables|MeshSlice|GetSliced%20Mesh))))
  (bind (_min_2 _max_2) (Utilities|Struct|BreakBox (GeometryScript|MeshQueries|GetMeshBoundingBox (Variables|MeshSlice|GetSliced%20Mesh))))
  (bind _geometryscriptmeshplanesliceoptions (Utilities|Struct|MakeGeometryScriptMeshPlaneSliceOptions true -1 false 0.0))
  (bind _number_of_cuts_y (Variables|MeshSlice|GetNumber%20of%20Cuts%20Y))
  (bind _number_of_cuts_x (Variables|MeshSlice|GetNumber%20of%20Cuts%20X))
  (Variables|MeshSlice|SetSliced%20Mesh Mesh)
  (if (and (<= Number%20of%20Cuts%20X 0) (<= Number%20of%20Cuts%20Y 0))
    (return Mesh)
    (else
      (Utilities|FlowControl|Sequence
        (:then_0
          (if (>= _number_of_cuts_x 1)
            (for _index (range _number_of_cuts_x)
              (bind _returnvalue (GeometryScript|Booleans|ApplyMeshPlaneSlice (Variables|MeshSlice|GetSliced%20Mesh) (Math|Transform|MakeTransform (Math|Vector|MakeVector (+ (* (/ (- (.x _max_2) (.x _min_2)) (+ (Variables|MeshSlice|GetNumber%20of%20Cuts%20X) 1)) _index) (.x _min_2))) "90.000000, 0, 0") _geometryscriptmeshplanesliceoptions))
              (Variables|MeshSlice|SetSliced%20Mesh _returnvalue))))
        (:then_1
          (if (>= _number_of_cuts_y 1)
            (for _index_2 (range _number_of_cuts_y)
              (bind _returnvalue_2 (GeometryScript|Booleans|ApplyMeshPlaneSlice (Variables|MeshSlice|GetSliced%20Mesh) (Math|Transform|MakeTransform (Math|Vector|MakeVector 0.0 0.0 (+ (* (/ (- (.z _max) (.z _min)) (+ (Variables|MeshSlice|GetNumber%20of%20Cuts%20Y) 1)) _index_2) (.z _min)))) _geometryscriptmeshplanesliceoptions))
              (Variables|MeshSlice|SetSliced%20Mesh _returnvalue_2))))
        (:then_2
          (return (Variables|MeshSlice|GetSliced%20Mesh)))))))

(fn MakeDisplaceVector
  (params (Normal :type Vector) (VtxColor :type Linear%20Color%20Structure) (Magnitude :type Float%20%28double-precision%29))
  (return (* Normal Magnitude)))

(fn MakeGuideMeshDataLists
  (bind _positionlist (GeometryScript|MeshQueries|GetAllVertexPositions (Variables|GuideMeshData|GetGuideMesh)))
  (Variables|GuideMeshData|SetGuideMeshPosition%20List _positionlist)
  (bind _normallist (GeometryScript|Normals|GetMeshPerVertexNormals (Variables|GuideMeshData|GetGuideMesh)))
  (Variables|GuideMeshData|SetGuideMeshNormalList _normallist)
  (for _index (range (+ (GeometryScript|ListUtils|GetVectorListLastIndex _positionlist) 1))
    (Utilities|Array|Add (Variables|GuideMeshData|GetGuideMeshPosition) (* (Math|Vector|MakeVector (Math|Conversions|ToFloat(Integer) (Math|Float|Ceil (.x (/ (.ReturnValue (GeometryScript|ListUtils|GetVectorListItem (Variables|GuideMeshData|GetGuideMeshPosition%20List) _index)) (Math|Vector|MakeVector (Variables|Default|GetTolerance) (Variables|Default|GetTolerance) (Variables|Default|GetTolerance)))))) (Math|Conversions|ToFloat(Integer) (Math|Float|Ceil (.y (/ (.ReturnValue (GeometryScript|ListUtils|GetVectorListItem (Variables|GuideMeshData|GetGuideMeshPosition%20List) _index)) (Math|Vector|MakeVector (Variables|Default|GetTolerance) (Variables|Default|GetTolerance) (Variables|Default|GetTolerance)))))) (Math|Conversions|ToFloat(Integer) (Math|Float|Ceil (.z (/ (.ReturnValue (GeometryScript|ListUtils|GetVectorListItem (Variables|GuideMeshData|GetGuideMeshPosition%20List) _index)) (Math|Vector|MakeVector (Variables|Default|GetTolerance) (Variables|Default|GetTolerance) (Variables|Default|GetTolerance))))))) (Math|Vector|MakeVector (Variables|Default|GetTolerance) (Variables|Default|GetTolerance) (Variables|Default|GetTolerance))))))

(fn MakeGuideMeshTriangleListfromUV1Mask
  (params (TriangleList :type Triangle%20List%20Structure) (UVList :type UV%20List%20Structure))
  (for _index (range (+ (GeometryScript|ListUtils|GetIndexListLastIndex (Variables|PreTesselationData|GetTriangleIDList_PreTesselation)) 1))
    (if (< (.Y (Math|Vector2D|BreakVector2D (.ReturnValue (GeometryScript|ListUtils|GetUVListItem UVList (Utilities|Struct|BreakIntVector (.ReturnValue (GeometryScript|ListUtils|GetTriangleListItem TriangleList _index))))))) 0)
      (Utilities|Array|Add (Variables|GuideMeshData|GetTriangleList_GuideMesh) _index)))
  (bind _indexlist (GeometryScript|ListUtils|ConvertArraytoIndexList (Variables|GuideMeshData|GetTriangleList_GuideMesh) "Triangle"))
  (Variables|GuideMeshData|SetGuideMeshTriangleList _indexlist))

(fn MakeGuideMeshTriangleListfromVertexColorMask
  (params (TriangleList :type Triangle%20List%20Structure))
  (bind _colorlist (GeometryScript|VertexColor|GetMeshPerVertexColors (Variables|Default|GetTargetMesh) false))
  (Variables|PreTesselationData|SetVertexColorList _colorlist)
  (for _index (range (+ (GeometryScript|ListUtils|GetIndexListLastIndex (Variables|PreTesselationData|GetTriangleIDList_PreTesselation)) 1))
    (if (> (Math|Color|BreakColor (.ReturnValue (GeometryScript|ListUtils|GetColorListItem (Variables|PreTesselationData|GetVertexColorList) (Utilities|Struct|BreakIntVector (.ReturnValue (GeometryScript|ListUtils|GetTriangleListItem TriangleList _index)))))) 0.5)
      (Utilities|Array|Add (Variables|GuideMeshData|GetTriangleList_GuideMesh) _index)))
  (bind _indexlist (GeometryScript|ListUtils|ConvertArraytoIndexList (Variables|GuideMeshData|GetTriangleList_GuideMesh) "Triangle"))
  (Variables|GuideMeshData|SetGuideMeshTriangleList _indexlist))

(fn MakeTriangleListforTesselation
  (bind _triangleidlist_pretesselation (Variables|PreTesselationData|GetTriangleIDList_PreTesselation))
  (for _index (range (GeometryScript|ListUtils|GetIndexListLength _triangleidlist_pretesselation))
    (if (Math|Boolean|ORBoolean (and (== (Math|Float|Floor (Math|Vector2D|BreakVector2D (GeometryScript|MeshQueries|GetTriangleUVs (Variables|Default|GetTargetMesh) 1 _index))) 0) (> (Variables|Default|GetMagnitude_1) 0)) (and (== (Math|Float|Floor (Math|Vector2D|BreakVector2D (GeometryScript|MeshQueries|GetTriangleUVs (Variables|Default|GetTargetMesh) 1 _index))) 1) (> (Variables|Default|GetMagnitude_2) 0)) (and (== (Math|Float|Floor (Math|Vector2D|BreakVector2D (GeometryScript|MeshQueries|GetTriangleUVs (Variables|Default|GetTargetMesh) 1 _index))) 2) (> (Variables|Default|GetMagnitude_3) 0)) (and (== (Math|Float|Floor (Math|Vector2D|BreakVector2D (GeometryScript|MeshQueries|GetTriangleUVs (Variables|Default|GetTargetMesh) 1 _index))) 3) (> (Variables|Default|GetMagnitude_4) 0)))
      (Utilities|Array|Add (Variables|PreTesselationData|GetTriangleList_Displ) (.ReturnValue (GeometryScript|ListUtils|GetIndexListItem _triangleidlist_pretesselation _index)))))
  (bind _indexlist (GeometryScript|ListUtils|ConvertArraytoIndexList (Variables|PreTesselationData|GetTriangleList_Displ) "Triangle"))
  (Variables|PreTesselationData|SetTriangle%20IDListForTesselation _indexlist))

(fn StealNormal
  (bind _targetmesh (Variables|Default|GetTargetMesh))
  (bind _returnvalue (GeometryScript|MeshQueries|GetNumVertexIDs _targetmesh))
  (bind _positionlist (GeometryScript|MeshQueries|GetAllVertexPositions _targetmesh))
  (Development|PrintString (Utilities|String|Append (Utilities|String|ToString(Integer) _returnvalue) "Number of loops"))
  (for _index (range _returnvalue)
    (Utilities|FlowControl|Sequence
      (:then_0
        ;[1 0.7668 0] debug
        (if (Variables|Default|GetDebug)
          (bind _returnvalue_2 (AddComponent|NODE_AddTextRenderComponent-0 :RelativeTransform (Math|Transform|MakeTransform (.ReturnValue (GeometryScript|ListUtils|GetVectorListItem _positionlist _index)))))
          (Rendering|Components|TextRender|SetText _returnvalue_2 (Utilities|Text|ToText(Integer) _index))))
      (:then_1
        ;[1 1 1] If there is a better way to Smooth a Normal of split vertices remove this
        (if (Variables|Options|GetUseAverageNormals)
          ;[1 1 1] get multiple normals for split vertices - slower but needed for split vertex corners - Might be faster if we could recompute smooth normals on the Guide Mesh
          (CallFunction|FindNormalfromGuidemeshMultipleSamples (.ReturnValue (GeometryScript|ListUtils|GetVectorListItem _positionlist _index)))
          (else
            ;[1 0.04114 0.03119] Find if vertex matches the position of a guide mesh vertex and if so steal its normal
            (CallFunction|FindlNormalfromGuidemesh (.ReturnValue (GeometryScript|ListUtils|GetVectorListItem _positionlist _index)))))))))

(fn TesselationfromTriangleList
  (bind _triangle_idlistfortesselation (Variables|PreTesselationData|GetTriangle%20IDListForTesselation))
  (bind _tessellationlevel (Variables|Default|GetTessellationLevel))
  (bind _returnvalue (DynamicMeshActor|AllocateComputeMesh))
  (bind (_storetosubmeshout _returnvalue_2) (GeometryScript|Decomposition|GetSubMeshfromMesh (Variables|Default|GetTargetMesh) _returnvalue _triangle_idlistfortesselation))
  (GeometryScript|MeshEdits|DeleteTrianglesfromMesh _returnvalue_2 _triangle_idlistfortesselation)
  (if (Variables|Options|GetPNTesseleation)
    (bind _returnvalue_3 (GeometryScript|Subdivide|ApplyPNTessellation _storetosubmeshout (struct :bRecomputeNormals true) _tessellationlevel)) ; causes cracks
    (GeometryScript|MeshEdits|AppendMesh :TargetMesh (Variables|Default|GetTargetMesh) :AppendMesh _returnvalue_3 :bDeferChangeNotifications false :AppendOptions "(CombineMode=EnableAllMatching)")
    (else
      (bind _returnvalue_4 (GeometryScript|Subdivide|ApplyUniformTessellation _storetosubmeshout _tessellationlevel))
      (GeometryScript|MeshEdits|AppendMesh :TargetMesh (Variables|Default|GetTargetMesh) :AppendMesh _returnvalue_4 :bDeferChangeNotifications false :AppendOptions "(CombineMode=EnableAllMatching)"))))

(fn ConstructionScript)
